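# Physics asset for the large male Constructor 01 player body, named as a set of cape colliders.
# Capsules on the pelvis, spine, clavicles and hands, plus a head sphere, form the collision volume; arm, leg, foot and neck bodies have no shapes.
# Constraints link bodies with all swing and twist motion free, and omit the head, left thigh-to-pelvis and left calf-to-thigh links.
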PhysicsAsset /Game/Characters/Player/Male/Large/Bodies/M_Lrg_Constructor_01/Mesh/M_LRG_Constuctor_CapeColliders1
bodies {
 calf_l []
 calf_r []
 clavicle_l [{shapeType=Capsule, length=11.3526, radius=12.6279, shapeName=clavicle_l_capsule, center=<10.2413 -1.5523 -1.15>, rotation=<rot 62.6575 0 0>}]
 clavicle_r [{
   shapeType=Capsule
   length=11.3526
   radius=12.6279
   shapeName=clavicle_l_capsule
   center=<-10.2413 1.5523 1.15>
   rotation=<rot 62.6575 0 -180>
  }]
 foot_l []
 foot_r []
 hand_l [{shapeType=Capsule, length=5.0012, radius=7.2893, shapeName=hand_l_capsule, center=<7.1732 -2.8011 -0.0005>, rotation=<rot 90 0 0>}]
 hand_r [{shapeType=Capsule, length=5.0012, radius=7.2893, shapeName=hand_l_capsule, center=<-7.1732 2.8011 0.0005>, rotation=<rot 90 0 180>}]
 head [{shapeType=Sphere, radius=19.5, center=<9.2893 2.7894 -0.0001>}]
 lowerarm_l []
 lowerarm_r []
 neck_02 []
 pelvis [{shapeType=Capsule, length=13.4832, radius=17.0793, center=<-4.8511 0 0>}]
 spine_01 [{shapeType=Capsule, length=7.875, radius=17.1189, center=<0 1.3296 0>}]
 spine_02 [{shapeType=Capsule, length=7.4062, radius=17.5876, center=<0 0.4678 -0.0001>}]
 spine_03 [{shapeType=Capsule, length=6.5, radius=18.4939, center=<0 -0.4292 0>}]
 spine_04 [{shapeType=Capsule, length=23.4757, radius=17.4682, center=<-0.0424 -1.5349 0.0001>, rotation=<rot 0 3.3589 0>}]
 spine_05 [{shapeType=Capsule, length=14.3855, radius=15.242, center=<0.0615 -1.9195 0.0056>}]
 thigh_l []
 thigh_r []
 upperarm_l []
 upperarm_r []
}
constraints [
 {
  bone1Name=upperarm_l
  bone2Name=clavicle_l
  swing1LimitDegrees=45
  swing1Motion=Free
  swing2LimitDegrees=45
  swing2Motion=Free
  twistLimitDegrees=45
  twistMotion=Free
 }
 {
  bone1Name=lowerarm_l
  bone2Name=upperarm_l
  swing1LimitDegrees=45
  swing1Motion=Free
  swing2LimitDegrees=45
  swing2Motion=Free
  twistLimitDegrees=45
  twistMotion=Free
 }
 {
  bone1Name=spine_05
  bone2Name=spine_04
  swing1LimitDegrees=45
  swing1Motion=Free
  swing2LimitDegrees=45
  swing2Motion=Free
  twistLimitDegrees=45
  twistMotion=Free
 }
 {
  bone1Name=neck_02
  bone2Name=spine_05
  swing1LimitDegrees=45
  swing1Motion=Free
  swing2LimitDegrees=45
  swing2Motion=Free
  twistLimitDegrees=45
  twistMotion=Free
 }
 {
  bone1Name=spine_04
  bone2Name=spine_03
  swing1LimitDegrees=45
  swing1Motion=Free
  swing2LimitDegrees=45
  swing2Motion=Free
  twistLimitDegrees=45
  twistMotion=Free
 }
 {
  bone1Name=clavicle_l
  bone2Name=spine_05
  swing1LimitDegrees=45
  swing1Motion=Free
  swing2LimitDegrees=45
  swing2Motion=Free
  twistLimitDegrees=45
  twistMotion=Free
 }
 {
  bone1Name=hand_l
  bone2Name=lowerarm_l
  swing1LimitDegrees=45
  swing1Motion=Free
  swing2LimitDegrees=45
  swing2Motion=Free
  twistLimitDegrees=45
  twistMotion=Free
 }
 {
  bone1Name=foot_l
  bone2Name=calf_l
  swing1LimitDegrees=45
  swing1Motion=Free
  swing2LimitDegrees=45
  swing2Motion=Free
  twistLimitDegrees=45
  twistMotion=Free
 }
 {
  bone1Name=spine_03
  bone2Name=spine_02
  swing1LimitDegrees=45
  swing1Motion=Free
  swing2LimitDegrees=45
  swing2Motion=Free
  twistLimitDegrees=45
  twistMotion=Free
 }
 {
  bone1Name=spine_01
  bone2Name=pelvis
  swing1LimitDegrees=45
  swing1Motion=Free
  swing2LimitDegrees=45
  swing2Motion=Free
  twistLimitDegrees=45
  twistMotion=Free
 }
 {
  bone1Name=spine_02
  bone2Name=spine_01
  swing1LimitDegrees=45
  swing1Motion=Free
  swing2LimitDegrees=45
  swing2Motion=Free
  twistLimitDegrees=45
  twistMotion=Free
 }
 {
  bone1Name=clavicle_r
  bone2Name=spine_05
  swing1LimitDegrees=45
  swing1Motion=Free
  swing2LimitDegrees=45
  swing2Motion=Free
  twistLimitDegrees=45
  twistMotion=Free
 }
 {
  bone1Name=upperarm_r
  bone2Name=clavicle_r
  swing1LimitDegrees=45
  swing1Motion=Free
  swing2LimitDegrees=45
  swing2Motion=Free
  twistLimitDegrees=45
  twistMotion=Free
 }
 {
  bone1Name=lowerarm_r
  bone2Name=upperarm_r
  swing1LimitDegrees=45
  swing1Motion=Free
  swing2LimitDegrees=45
  swing2Motion=Free
  twistLimitDegrees=45
  twistMotion=Free
 }
 {
  bone1Name=hand_r
  bone2Name=lowerarm_r
  swing1LimitDegrees=45
  swing1Motion=Free
  swing2LimitDegrees=45
  swing2Motion=Free
  twistLimitDegrees=45
  twistMotion=Free
 }
 {
  bone1Name=thigh_r
  bone2Name=pelvis
  swing1LimitDegrees=45
  swing1Motion=Free
  swing2LimitDegrees=45
  swing2Motion=Free
  twistLimitDegrees=45
  twistMotion=Free
 }
 {
  bone1Name=calf_r
  bone2Name=thigh_r
  swing1LimitDegrees=45
  swing1Motion=Free
  swing2LimitDegrees=45
  swing2Motion=Free
  twistLimitDegrees=45
  twistMotion=Free
 }
 {
  bone1Name=foot_r
  bone2Name=calf_r
  swing1LimitDegrees=45
  swing1Motion=Free
  swing2LimitDegrees=45
  swing2Motion=Free
  twistLimitDegrees=45
  twistMotion=Free
 }
]
