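# Animation Blueprint for the large male Constructor body that adds secondary physics motion to its dangling gear and clothing.
# On initialization it caches the owner's skeletal mesh component, whose pose it copies each update.
# That pose passes through chained Anim Dynamics nodes for pouches, flap, wrench, arm sign, straps and sleeves.
# Each uses tight per-bone limits, most with reduced gravity, and the sleeves blend at partial alpha.
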AnimBlueprint /Game/Characters/Player/Male/Large/Bodies/M_Lrg_Constructor_01/Mesh/SK_M_LRG_Constructor_01_Skeleton_AnimBP
target_skeleton=&/SK_M_LRG_Constructor_01_Skeleton
graphs [
 AnimGraph {nodes [
   n1 = Get MeshToCopy() -> (MeshToCopy:Skeletal Mesh Component Object Reference)
   n2 = Anim Dynamics - Bone: dyn_pouch_l(ComponentPose=Local To Component(LocalPose=Copy Pose From Mesh(SourceMeshComponent=@n1)), Alpha=1, ExternalForce="0,0,0") {
    bGravityOverrideInSimSpace=true
    gravityScale=0.1
    alphaBoolBlend {blendOption=Linear}
    boundBone {boneName=dyn_pouch_l}
    physicsBodyDefinitions [{
      sphereCollisionRadius=0
      boundBone {boneName=dyn_pouch_l}
      boxExtents=<12 20 7>
      constraintSetup {
       angularTargetAxis=AxisY
       twistAxis=AxisY
       angularLimitsMax=<15 0 0>
       angularLimitsMin=<-6 -10 0>
       angularTarget=<0 0 0>
       linearAxesMax=<0 0 2>
      }
      localJointOffset=<0 0 -10>
     }]
   }
   n3 = Anim Dynamics - Bone: dyn_pouch_flap_l(ComponentPose=@n2, Alpha=1, ExternalForce="0,0,0") {
    bGravityOverrideInSimSpace=true
    gravityScale=0.05
    numSolverIterationsPostUpdate=4
    numSolverIterationsPreUpdate=16
    alphaBoolBlend {blendOption=Linear}
    boundBone {boneName=dyn_pouch_flap_l}
    physicsBodyDefinitions [{
      sphereCollisionRadius=0
      boundBone {boneName=dyn_pouch_flap_l}
      boxExtents=<4 5 1>
      constraintSetup {angularTargetAxis=AxisY, twistAxis=AxisY, angularLimitsMax=<0 0 0>, angularLimitsMin=<0 0 0>, angularTarget=<0 0 0>, linearAxesMax=<0 0 8>}
      localJointOffset=<-1 0 0.1>
     }]
    relativeSpaceBone {boneName=pelvis}
   }
   n4 = Anim Dynamics - Bone: dyn_pouch_f(ComponentPose=@n3, Alpha=1, ExternalForce="0,0,0") {
    bGravityOverrideInSimSpace=true
    gravityScale=0.66
    numSolverIterationsPostUpdate=4
    numSolverIterationsPreUpdate=16
    alphaBoolBlend {blendOption=Linear}
    boundBone {boneName=dyn_pouch_f}
    physicsBodyDefinitions [{
      sphereCollisionRadius=0
      boundBone {boneName=dyn_pouch_f}
      boxExtents=<10 10 10>
      constraintSetup {angularTargetAxis=AxisY, twistAxis=AxisY, angularLimitsMax=<0 0 0>, angularLimitsMin=<0 0 0>, angularTarget=<0 0 0>, linearAxesMax=<0 0 10>}
      localJointOffset=<0 0 -2>
     }]
   }
   n5 = Anim Dynamics - Bone: dyn_wrench(ComponentPose=@n4, Alpha=1, ExternalForce="0,0,0") {
    bGravityOverrideInSimSpace=true
    alphaBoolBlend {blendOption=Linear}
    boundBone {boneName=dyn_wrench}
    physicsBodyDefinitions [{
      sphereCollisionRadius=0
      boundBone {boneName=dyn_wrench}
      boxExtents=<20 20 20>
      constraintSetup {
       angularTargetAxis=AxisY
       twistAxis=AxisY
       angularLimitsMax=<10 0 0>
       angularLimitsMin=<-10 0 0>
       angularTarget=<0 0 0>
       linearAxesMax=<0 0 5>
      }
      localJointOffset=<0 0 -15>
     }]
   }
   n6 = Anim Dynamics - Bone: dyn_arm_sign(ComponentPose=@n5, Alpha=1, ExternalForce="0,0,0") {
    angularBiasOverride=1
    bGravityOverrideInSimSpace=true
    bOverrideAngularBias=true
    numSolverIterationsPostUpdate=4
    numSolverIterationsPreUpdate=16
    alphaBoolBlend {blendOption=Linear}
    boundBone {boneName=dyn_arm_sign}
    physicsBodyDefinitions [{
      sphereCollisionRadius=0
      boundBone {boneName=dyn_arm_sign}
      boxExtents=<10 10 10>
      constraintSetup {angularLimitsMax=<3 0 0>, angularLimitsMin=<-3 0 0>, angularTarget=<0 0 0>}
      localJointOffset=<0 0 0>
     }]
   }
   n7 = Anim Dynamics - Bone: dyn_strap_l(ComponentPose=@n6, Alpha=1, ExternalForce="0,0,0") {
    angularBiasOverride=1
    bGravityOverrideInSimSpace=true
    bOverrideAngularBias=true
    bUsePlanarLimit=false
    gravityScale=0.1
    numSolverIterationsPostUpdate=4
    numSolverIterationsPreUpdate=16
    alphaBoolBlend {blendOption=Linear}
    boundBone {boneName=dyn_strap_l}
    physicsBodyDefinitions [{
      sphereCollisionRadius=0
      boundBone {boneName=dyn_strap_l}
      boxExtents=<5 5 5>
      constraintSetup {angularTargetAxis=AxisY, twistAxis=AxisY, angularLimitsMax=<0 0 0>, angularLimitsMin=<0 0 0>, angularTarget=<0 0 0>, linearAxesMax=<0 0 1>}
      localJointOffset=<0 0 -5>
     }]
   }
   n8 = Anim Dynamics - Bone: dyn_strap_r(ComponentPose=@n7, Alpha=1, ExternalForce="0,0,0") {
    angularBiasOverride=1
    bGravityOverrideInSimSpace=true
    bOverrideAngularBias=true
    bUsePlanarLimit=false
    gravityScale=0.1
    numSolverIterationsPostUpdate=4
    numSolverIterationsPreUpdate=16
    alphaBoolBlend {blendOption=Linear}
    boundBone {boneName=dyn_strap_r}
    physicsBodyDefinitions [{
      sphereCollisionRadius=0
      boundBone {boneName=dyn_strap_r}
      boxExtents=<5 5 5>
      constraintSetup {angularTargetAxis=AxisY, twistAxis=AxisY, angularLimitsMax=<0 0 0>, angularLimitsMin=<0 0 0>, angularTarget=<0 0 0>, linearAxesMax=<0 0 1>}
      localJointOffset=<0 0 -5>
     }]
   }
   n9 = Anim Dynamics - Bone: dyn_sleeve_r(ComponentPose=@n8, Alpha=.8, ExternalForce="0,0,0") {
    angularBiasOverride=1
    bGravityOverrideInSimSpace=true
    bOverrideAngularBias=true
    bUsePlanarLimit=false
    gravityScale=0.1
    numSolverIterationsPostUpdate=4
    numSolverIterationsPreUpdate=16
    alphaBoolBlend {blendOption=Linear}
    boundBone {boneName=dyn_sleeve_r}
    physicsBodyDefinitions [{
      sphereCollisionRadius=0
      boundBone {boneName=dyn_sleeve_r}
      boxExtents=<5 5 5>
      constraintSetup {angularLimitsMax=<0 0 0>, angularLimitsMin=<0 0 0>, angularTarget=<0 0 0>, linearAxesMax=<2.5 0.001 0.001>, linearAxesMin=<0 -0.001 -0.001>}
      localJointOffset=<-5 0 0>
     }]
   }
   n10 = Anim Dynamics - Bone: dyn_sleeve_l(ComponentPose=@n9, Alpha=.8, ExternalForce="0,0,0") {
    angularBiasOverride=1
    bGravityOverrideInSimSpace=true
    bOverrideAngularBias=true
    bUsePlanarLimit=false
    gravityScale=0.1
    numSolverIterationsPostUpdate=4
    numSolverIterationsPreUpdate=16
    alphaBoolBlend {blendOption=Linear}
    boundBone {boneName=dyn_sleeve_l}
    physicsBodyDefinitions [{
      sphereCollisionRadius=0
      boundBone {boneName=dyn_sleeve_l}
      boxExtents=<5 5 5>
      constraintSetup {angularLimitsMax=<0 0 0>, angularLimitsMin=<0 0 0>, angularTarget=<0 0 0>, linearAxesMax=<2.5 0.001 0.001>, linearAxesMin=<0 -0.001 -0.001>}
      localJointOffset=<-5 0 0>
     }]
   }
   Output Pose(Result=Component To Local(ComponentPose=@n10)) {layerGroup=None}
  ]}
 EventGraph {nodes [
   Event BlueprintInitializeAnimation()
   > Set MeshToCopy(MeshToCopy:Skeletal Mesh Component Object Reference=GetOwnerSkeletalMeshComponent()) -> (Output_Get:Skeletal Mesh Component Object Reference)
  ]}
]
variables [MeshToCopy {type=SkeletalMeshComponent}]
